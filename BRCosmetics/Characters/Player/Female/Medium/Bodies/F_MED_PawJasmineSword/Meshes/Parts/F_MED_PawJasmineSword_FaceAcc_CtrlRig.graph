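# Corrective Control Rig for the PawJasmineSword female cosmetic's face accessory part, procedurally deforming hood bones as head and hood move.
# A four-step sequence first rotates hood_up_m by remapped head-to-HeadBoneAim distance.
# Spherical pose readers on head (relative to spine_05) and hood_dyn then weight additive offsets on middle and side hood bones.
# Finally, a rotation constraint drives hood_up_deform_base, and a hood_up_m pose reader translates hood_up_deform and hood_down.
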ControlRigBlueprint /BRCosmetics/Characters/Player/Female/Medium/Bodies/F_MED_PawJasmineSword/Meshes/Parts/F_MED_PawJasmineSword_FaceAcc_CtrlRig
graphs [
 RigVMFunctionLibrary {links [], nodes []}
 RigVMFunction_Sequence_1_ContainedGraph {nodes [
   n1 = RigVMFunctionEntryNode()
   RigVMFunction_Sequence = &RigVMFunction_Sequence(ExecuteContext=@n1.ExecuteContext)
   n2 = &RigVMFunction_Sequence(ExecuteContext=@RigVMFunction_Sequence.B)
   n3 = &RigVMFunction_Sequence(ExecuteContext=@n2.B)
   RigVMFunctionReturnNode(A=@RigVMFunction_Sequence.A, B=@n2.A, C=@n3.A, D=@n3.B)
  ]}
 RigVMModel {nodes [
   RigUnit_BeginExecution = &RigUnit_BeginExecution()
   RigVMFunction_Sequence = RigVMAggregateNode(ExecuteContext=@RigUnit_BeginExecution.ExecutePin)
   RigUnit_GetTransform = &RigUnit_GetTransform(Item=(Type=Bone,Name="HeadBoneAim"), bInitial=False)
   RigVMFunction_MathVectorDistance = &RigVMFunction_MathVectorDistance(
    B=@RigUnit_GetTransform.Transform.Translation
    A=&RigUnit_GetTransform(Item=(Type=Bone,Name="head"), bInitial=False).Transform.Translation
   )
   Remap = &RigVMFunction_MathDoubleRemap(SourceMaximum=7, SourceMinimum=11.5, TargetMaximum=1, TargetMinimum=0, Value=@RigVMFunction_MathVectorDistance.Result)
   &RigUnit_OffsetTransformForItem(
    Item=(Type=Bone,Name="hood_up_m")
    OffsetTransform=(Rotation=(X=-0.0118,Y=-0.1039,Z=-0.1126,W=0.9881),Translation=(X=0,Y=0,Z=0),Scale3D=(X=1,Y=1,Z=1))
    bPropagateToChildren=True
    Weight=@Remap.Result
    ExecutePin=@RigVMFunction_Sequence.A
   )
   n1 = &RigUnit_SphericalPoseReader(
    ActiveRegionScaleFactors=(PositiveWidth=1,NegativeWidth=1,PositiveHeight=1,NegativeHeight=1)
    ActiveRegionSize=0.25
    Debug=(bDrawDebug=false,bDraw2D=False,bDrawLocalAxes=False,DebugScale=10,DebugSegments=20,DebugThickness=0.1)
    DriverAxis=(X=1,Y=0,Z=0)
    DriverItem=(Type=Bone,Name="head")
    FalloffRegionScaleFactors=(PositiveWidth=1.7,NegativeWidth=1,PositiveHeight=1,NegativeHeight=1)
    FalloffSize=0.5
    FlipHeightScaling=False
    FlipWidthScaling=False
    OptionalParentItem=(Type=Bone,Name="spine_05")
    RotationOffset=(X=-90,Y=0,Z=-35)
    ExecutePin=@RigVMFunction_Sequence.B
   )
   > RigUnit_OffsetTransformForItem = &RigUnit_OffsetTransformForItem(
    Item=(Type=Bone,Name="hood_up_m")
    OffsetTransform=(Rotation=(X=0.0616,Y=-0.0616,Z=-0.7044,W=0.7044),Translation=(X=0,Y=0,Z=0),Scale3D=(X=1,Y=1,Z=1))
    bPropagateToChildren=True
    Weight=@n1.OutputParam
   )
   > &RigUnit_OffsetTransformForItem(
    Item=(Type=Bone,Name="hood_deform_r")
    OffsetTransform=(Rotation=(X=0,Y=0,Z=-0.4617,W=0.887),Translation=(X=0,Y=0,Z=0),Scale3D=(X=1,Y=1,Z=1))
    bPropagateToChildren=True
    Weight=@n1.OutputParam
   )
   > &RigUnit_OffsetTransformForItem(
    Item=(Type=Bone,Name="hood_deform_l")
    OffsetTransform=(Rotation=(X=0.2432,Y=-0.0885,Z=-0.3304,W=0.9077),Translation=(X=0,Y=0,Z=0),Scale3D=(X=1,Y=1,Z=1))
    bPropagateToChildren=True
    Weight=@n1.OutputParam
   )
   > n2 = &RigUnit_SphericalPoseReader(
    ActiveRegionScaleFactors=(PositiveWidth=1,NegativeWidth=1,PositiveHeight=1,NegativeHeight=1)
    ActiveRegionSize=0.3
    Debug=(bDrawDebug=false,bDraw2D=False,bDrawLocalAxes=False,DebugScale=10,DebugSegments=20,DebugThickness=0.1)
    DriverAxis=(X=1,Y=0,Z=0)
    DriverItem=(Type=Bone,Name="head")
    FalloffRegionScaleFactors=(PositiveWidth=1,NegativeWidth=1,PositiveHeight=1,NegativeHeight=1)
    FalloffSize=0.5
    FlipHeightScaling=False
    FlipWidthScaling=False
    OptionalParentItem=(Type=Bone,Name="spine_05")
    RotationOffset=(X=0,Y=0,Z=0)
   )
   > &RigUnit_OffsetTransformForItem(
    Item=(Type=Bone,Name="hood_deform_r")
    OffsetTransform=(Rotation=(X=-0.3339,Y=0.074,Z=-0.2034,W=0.9174),Translation=(X=2.5,Y=0,Z=0),Scale3D=(X=1,Y=1,Z=1))
    bPropagateToChildren=True
    Weight=@n2.OutputParam
   )
   > n3 = &RigUnit_SphericalPoseReader(
    ActiveRegionScaleFactors=(PositiveWidth=1,NegativeWidth=1,PositiveHeight=1,NegativeHeight=1)
    ActiveRegionSize=0.2
    Debug=(bDrawDebug=false,bDraw2D=False,bDrawLocalAxes=False,DebugScale=10,DebugSegments=20,DebugThickness=0.1)
    DriverAxis=(X=0,Y=1,Z=0)
    DriverItem=(Type=Bone,Name="head")
    FalloffRegionScaleFactors=(PositiveWidth=1,NegativeWidth=1,PositiveHeight=1,NegativeHeight=1)
    FalloffSize=0.5
    FlipHeightScaling=False
    FlipWidthScaling=False
    OptionalParentItem=(Type=Bone,Name="spine_05")
    RotationOffset=(X=180,Y=0,Z=0)
   )
   > &RigUnit_OffsetTransformForItem(
    Item=(Type=Bone,Name="hood_up_m")
    OffsetTransform=(Rotation=(X=-0.1722,Y=0.0227,Z=-0.1285,W=0.9764),Translation=(X=0,Y=0,Z=0),Scale3D=(X=1,Y=1,Z=1))
    bPropagateToChildren=True
    Weight=@n3.OutputParam
   )
   > &RigUnit_OffsetTransformForItem(
    Item=(Type=Bone,Name="hood_deform_l")
    OffsetTransform=(Rotation=(X=0.1261,Y=-0.0338,Z=-0.2566,W=0.9577),Translation=(X=0,Y=0,Z=0),Scale3D=(X=1,Y=1,Z=1))
    bPropagateToChildren=True
    Weight=@n3.OutputParam
   )
   > n4 = &RigUnit_SphericalPoseReader(
    ActiveRegionScaleFactors=(PositiveWidth=1,NegativeWidth=1,PositiveHeight=1,NegativeHeight=1)
    ActiveRegionSize=0.2
    Debug=(bDrawDebug=false,bDraw2D=False,bDrawLocalAxes=False,DebugScale=10,DebugSegments=20,DebugThickness=0.1)
    DriverAxis=(X=0,Y=1,Z=0)
    DriverItem=(Type=Bone,Name="head")
    FalloffRegionScaleFactors=(PositiveWidth=1,NegativeWidth=1,PositiveHeight=1,NegativeHeight=1)
    FalloffSize=0.5
    FlipHeightScaling=False
    FlipWidthScaling=False
    OptionalParentItem=(Type=Bone,Name="spine_05")
    RotationOffset=(X=0,Y=0,Z=0)
   )
   > &RigUnit_OffsetTransformForItem(
    Item=(Type=Bone,Name="hood_up_m")
    OffsetTransform=(Rotation=(X=0,Y=0,Z=-0.1305,W=0.9914),Translation=(X=0,Y=0,Z=0),Scale3D=(X=1,Y=1,Z=1))
    bPropagateToChildren=True
    Weight=@n4.OutputParam
   )
   > &RigUnit_OffsetTransformForItem(
    Item=(Type=Bone,Name="hood_deform_r")
    OffsetTransform=(Rotation=(X=0,Y=0,Z=-0.2588,W=0.9659),Translation=(X=0,Y=0,Z=0),Scale3D=(X=1,Y=1,Z=1))
    bPropagateToChildren=True
    Weight=@n4.OutputParam
   )
   n5 = &RigUnit_SphericalPoseReader(
    ActiveRegionScaleFactors=(PositiveWidth=1,NegativeWidth=1,PositiveHeight=1,NegativeHeight=1)
    ActiveRegionSize=0.1
    Debug=(bDrawDebug=false,bDraw2D=False,bDrawLocalAxes=False,DebugScale=15,DebugSegments=20,DebugThickness=0.1)
    DriverAxis=(X=1,Y=0,Z=0)
    DriverItem=(Type=Bone,Name="hood_dyn")
    FalloffRegionScaleFactors=(PositiveWidth=1,NegativeWidth=1,PositiveHeight=1,NegativeHeight=1)
    FalloffSize=0.5
    FlipHeightScaling=False
    FlipWidthScaling=False
    OptionalParentItem=(Type=Bone,Name="None")
    RotationOffset=(X=90,Y=0,Z=0)
    ExecutePin=@RigVMFunction_Sequence.C
   )
   > &RigUnit_OffsetTransformForItem(
    Item=(Type=Bone,Name="hood_up_m")
    OffsetTransform=(Rotation=(X=0,Y=0,Z=-0.6088,W=0.7934),Translation=(X=0,Y=0,Z=0),Scale3D=(X=1,Y=1,Z=1))
    bPropagateToChildren=True
    Weight=@n5.OutputParam
   )
   &RigUnit_RotationConstraintLocalSpaceOffset(
    AdvancedSettings=(InterpolationType=Average,RotationOrderForFilter=XZY)
    Child=(Type=Bone,Name="hood_up_deform_base")
    Filter=(bX=True,bY=True,bZ=True)
    Parents=((Item=(Type=Bone,Name="hood_dyn"),Weight=1.2),(Item=(Type=Bone,Name="hood_up_m"),Weight=0.7))
    Weight=1
    ExecutePin=@RigVMFunction_Sequence.D
   )
   > n6 = &RigUnit_SphericalPoseReader(
    ActiveRegionScaleFactors=(PositiveWidth=1,NegativeWidth=1,PositiveHeight=1,NegativeHeight=1)
    ActiveRegionSize=0.1
    Debug=(bDrawDebug=false,bDraw2D=False,bDrawLocalAxes=False,DebugScale=10,DebugSegments=20,DebugThickness=0.1)
    DriverAxis=(X=1,Y=0,Z=0)
    DriverItem=(Type=Bone,Name="hood_up_m")
    FalloffRegionScaleFactors=(PositiveWidth=1.3,NegativeWidth=1,PositiveHeight=1,NegativeHeight=1)
    FalloffSize=0.5
    FlipHeightScaling=False
    FlipWidthScaling=False
    OptionalParentItem=(Type=Bone,Name="hood_dyn")
    RotationOffset=(X=-90,Y=0,Z=-15)
   )
   > &RigUnit_OffsetTransformForItem(
    Item=(Type=Bone,Name="hood_up_deform")
    OffsetTransform=(Rotation=(X=0,Y=0,Z=0,W=1),Translation=(X=12,Y=0,Z=0),Scale3D=(X=1,Y=1,Z=1))
    bPropagateToChildren=True
    Weight=@n6.OutputParam
   )
   > &RigUnit_OffsetTransformForItem(
    Item=(Type=Bone,Name="hood_down")
    OffsetTransform=(Rotation=(X=0,Y=0,Z=0,W=1),Translation=(X=6,Y=-6,Z=0),Scale3D=(X=1,Y=1,Z=1))
    bPropagateToChildren=True
    Weight=@n6.OutputParam
   )
   &RigUnit_OffsetTransformForItem(
    ExecutePin=()
    Item=(Type=Bone,Name="hood_dyn")
    OffsetTransform=(Rotation=(X=0,Y=0,Z=0.4848,W=0.8746),Translation=(X=0,Y=0,Z=0),Scale3D=(X=1,Y=1,Z=1))
    Weight=1
    bPropagateToChildren=True
   )
  ]}
]
hierarchy [
 root {loc=<0 0 0>, rot=<0 0 0 1>}
 pelvis {parent=root, loc=<0 -0.5301 98.0571>, rot=<0.0127 0.707 -0.0127 -0.707>}
 spine_01 {parent=pelvis, loc=<6.1208 -0.9328 0>, rot=<0 0 0.034 -0.9994>}
 spine_02 {parent=spine_01, loc=<6.5997 -0.2095 0>, rot=<0 0 0.0238 -0.9997>}
 spine_03 {parent=spine_02, loc=<6.4914 -0.1237 0>, rot=<0 0 0.0149 -0.9999>}
 spine_04 {parent=spine_03, loc=<6.9598 0.0794 0>, rot=<0 0 -0.0132 -0.9999>}
 spine_05 {parent=spine_04, loc=<6.9987 0 0>, rot=<0 0 -0.0015 -1>}
 neck_01 {parent=spine_05, loc=<11.3138 0.1529 0>, rot=<0 0 -0.2003 -0.9797>}
 neck_02 {parent=neck_01, loc=<4.5627 0 0>, rot=<0 0 0.0504 -0.9987>}
 head {parent=neck_02, loc=<5.3914 0 0>, rot=<0 0 0.115 -0.9934>}
 MSTR_hairHead_l {parent=head, loc=<11.7763 -4.1718 -6.7099>, rot=<0.7777 -0.3809 -0.206 -0.4558>}
 hairHead_l {parent=MSTR_hairHead_l, loc=<0 0 0>, rot=<0 0 0 1>}
 MASTR_hair_A_r_deform_1 {parent=hairHead_l, loc=<-0.7116 -3.2449 1.4436>, rot=<0.0159 -0.0069 0.649 -0.7606>}
 hair_A_r_deform_1 {parent=MASTR_hair_A_r_deform_1, loc=<0 0 0>, rot=<0 0 0 1>}
 hair_A_r_deform_2 {parent=hair_A_r_deform_1, loc=<1.5459 0 0>, rot=<-0.0609 0.171 -0.3301 -0.9263>}
 MASTR_hair_B_r_deform_1 {parent=hairHead_l, loc=<1.8703 -3.0593 3.1556>, rot=<0.2957 -0.1722 0.029 -0.9392>}
 hair_B_r_deform_1 {parent=MASTR_hair_B_r_deform_1, loc=<0 0 0>, rot=<0 0 0 1>}
 hair_B_r_deform_2 {parent=hair_B_r_deform_1, loc=<1.6085 0 0>, rot=<0 0 -0.2017 -0.9794>}
 hair_B_r_deform_3 {parent=hair_B_r_deform_2, loc=<1.5433 0 0>, rot=<0 0 -0.1129 -0.9936>}
 MSTR_hair_C_r_deform_1 {parent=hairHead_l, loc=<1.5836 2.9165 2.7593>, rot=<0.9347 0.0447 0.0258 -0.3516>}
 hair_C_r_deform_1 {parent=MSTR_hair_C_r_deform_1, loc=<0 0 0>, rot=<0 0 0 1>}
 hair_C_r_deform_2 {parent=hair_C_r_deform_1, loc=<1.8732 0 0>, rot=<0.0575 -0.1021 -0.3124 -0.9427>}
 hair_C_r_deform_3 {parent=hair_C_r_deform_2, loc=<1.1187 0 0>, rot=<0 0 -0.1556 -0.9878>}
 MSTR_hair_D_r_deform_1 {parent=hairHead_l, loc=<1.0763 -2.9141 -2.846>, rot=<-0.3301 0.0502 0.1769 -0.9259>}
 hair_D_r_deform_1 {parent=MSTR_hair_D_r_deform_1, loc=<0 0 0>, rot=<0 0 0 1>}
 hair_D_r_deform_2 {parent=hair_D_r_deform_1, loc=<1.3035 0 0>, rot=<0.0327 0.0613 -0.0469 -0.9965>}
 hair_D_r_deform_3 {parent=hair_D_r_deform_2, loc=<0.969 0 0>, rot=<-0.0028 0.1001 0.0281 -0.9946>}
 MSTR_hair_E_r_deform_1 {parent=hairHead_l, loc=<1.5014 2.0081 -3.2217>, rot=<0.9226 0.0057 -0.0782 0.3778>}
 hair_E_r_deform_1 {parent=MSTR_hair_E_r_deform_1, loc=<0 0 0>, rot=<0 0 0 1>}
 hair_E_r_deform_2 {parent=hair_E_r_deform_1, loc=<1.5206 0 0>, rot=<0.0403 0.0126 -0.2014 -0.9786>}
 hair_E_r_deform_3 {parent=hair_E_r_deform_2, loc=<1.7624 0 0>, rot=<0.007 0.0469 -0.1382 -0.9893>}
 MSTR_hair_F_r_deform_1 {parent=hairHead_l, loc=<-0.2994 3.0869 1.1487>, rot=<-0.015 -0.0871 -0.5315 -0.8424>}
 hair_F_r_deform_1 {parent=MSTR_hair_F_r_deform_1, loc=<0 0 0>, rot=<0 0 0 1>}
 hair_F_r_deform_2 {parent=hair_F_r_deform_1, loc=<1.1738 0 0>, rot=<0 0 -0.2089 -0.9779>}
 hair_F_r_deform_3 {parent=hair_F_r_deform_2, loc=<0.9013 0.0001 0>, rot=<0.0003 -0.0019 -0.1405 -0.9901>}
 MSTR_hairHead_r {parent=head, loc=<11.7768 -4.1718 6.7099>, rot=<0.7777 -0.3809 0.206 0.4558>}
 hairHead_r {parent=MSTR_hairHead_r, loc=<0 0 0>, rot=<0 0 0 1>}
 MASTR_hair_E_l_deform_1 {parent=hairHead_r, loc=<1.3675 2.8769 2.4999>, rot=<0.9379 -0.0335 -0.0492 -0.3417>}
 hair_E_l_deform_1 {parent=MASTR_hair_E_l_deform_1, loc=<0 0 0>, rot=<0 0 0 1>}
 hair_E_l_deform_2 {parent=hair_E_l_deform_1, loc=<1.096 0 0>, rot=<-0.0022 0.0202 -0.0856 -0.9961>}
 MASTR_hair_B_l_deform_1 {parent=hairHead_r, loc=<0.2216 0.6928 2.6718>, rot=<0.6255 0.0234 0.2801 -0.7279>}
 hair_B_l_deform_1 {parent=MASTR_hair_B_l_deform_1, loc=<0 0 0>, rot=<0 0 0 1>}
 hair_B_l_deform_2 {parent=hair_B_l_deform_1, loc=<3.5 0 0>, rot=<0 0 -0.4688 -0.8833>}
 hair_B_l_deform_3 {parent=hair_B_l_deform_2, loc=<2.3089 0 0>, rot=<0 0 -0.2122 -0.9772>}
 MASTR_hair_C_l_deform_1 {parent=hairHead_r, loc=<2.8406 1.4606 -3.9726>, rot=<-0.8272 0.213 -0.0147 -0.5198>}
 hair_C_l_deform_1 {parent=MASTR_hair_C_l_deform_1, loc=<0 0 0>, rot=<0 0 0 1>}
 hair_C_l_deform_2 {parent=hair_C_l_deform_1, loc=<1.2588 0 0>, rot=<0.0129 -0.0201 -0.1387 -0.99>}
 MASTR_hair_D_l_deform_1 {parent=hairHead_r, loc=<-0.8261 2.8847 -1.9556>, rot=<0.8615 0.3518 -0.0409 0.3638>}
 hair_D_l_deform_1 {parent=MASTR_hair_D_l_deform_1, loc=<0 0 0>, rot=<0 0 0 1>}
 hair_D_l_deform_2 {parent=hair_D_l_deform_1, loc=<1.8685 0 0>, rot=<0 0 -0.1702 -0.9854>}
 hair_D_l_deform_3 {parent=hair_D_l_deform_2, loc=<1.1692 0 0>, rot=<0.0033 -0.0215 -0.1255 -0.9918>}
 MASTR_hair_A_l_deform_1 {parent=hairHead_r, loc=<-0.2361 -3.1791 0.4635>, rot=<0.1933 0.1249 0.2297 -0.9457>}
 hair_A_l_deform_1 {parent=MASTR_hair_A_l_deform_1, loc=<0 0 0>, rot=<0 0 0 1>}
 hair_A_l_deform_2 {parent=hair_A_l_deform_1, loc=<2.5 0 0>, rot=<0 0 -0.2092 -0.9779>}
 hair_A_l_deform_3 {parent=hair_A_l_deform_2, loc=<1.4002 0.0001 0>, rot=<-0.0078 0.0228 -0.3107 -0.9502>}
 MSTR_foreHair_l {parent=head, loc=<9.4643 8.5661 -0.498>, rot=<0.5776 -0.4641 -0.4249 -0.52>}
 foreHair_l_1 {parent=MSTR_foreHair_l, loc=<0 0 0>, rot=<0 0 0 1>}
 foreHair_l_2 {parent=foreHair_l_1, loc=<2.9355 0.0006 0>, rot=<-0.0029 -0.01 -0.4724 -0.8813>}
 foreHair_l_3 {parent=foreHair_l_2, loc=<3.7393 -0.0005 0>, rot=<-0.1105 -0.0623 -0.2762 -0.9527>}
 foreHair_l_4 {parent=foreHair_l_3, loc=<3.6942 0.0001 0>, rot=<0 0 0 1>}
 MSTR_foreHair_r {parent=head, loc=<9.4641 8.5661 0.498>, rot=<0.5956 -0.4785 0.4083 0.4996>}
 foreHair_r_1 {parent=MSTR_foreHair_r, loc=<0 0 0>, rot=<0 0 0 1>}
 foreHair_r_2 {parent=foreHair_r_1, loc=<2.9356 0 0>, rot=<0.0029 0.01 -0.4724 -0.8813>}
 foreHair_r_3 {parent=foreHair_r_2, loc=<3.7398 0 0>, rot=<0.1105 0.0623 -0.2762 -0.9527>}
 foreHair_r_4 {parent=foreHair_r_3, loc=<3.6938 0 0>, rot=<0 0 0 1>}
 head_pnt_mid {parent=head, loc=<2 -4.84 0>, rot=<0 0 0 1>}
 head_pnt_r {parent=head, loc=<2 -3 6>, rot=<0 0 0 1>}
 clavicle_l {parent=spine_05, loc=<5.4831 -0.1053 -3.2946>, rot=<0.0465 -0.7289 -0.009 -0.6829>}
 upperarm_l {parent=clavicle_l, loc=<9.8658 -1.2136 -0.4105>, rot=<0.0164 -0.3245 -0.0269 -0.9454>}
 upperarm_twist_01_l {parent=upperarm_l, loc=<7.2135 -0.1134 0>, rot=<0 0 0 1>}
 deform_upperarm_bk_base_l {parent=upperarm_twist_01_l, loc=<-2.133 -6.2124 0.4116>, rot=<0 0.0007 0.0191 -0.9998>}
 deform_upperarm_bk_l {parent=deform_upperarm_bk_base_l, loc=<0 0 0>, rot=<0 0 0 1>}
 deform_upperarm_fr_base_l {parent=upperarm_twist_01_l, loc=<-3.036 8.1024 -0.4102>, rot=<0 0.0007 0.0191 -0.9998>}
 deform_upperarm_fr_l {parent=deform_upperarm_fr_base_l, loc=<0 0 0>, rot=<0 0 0 1>}
 deform_upperarm_up_base_l {parent=upperarm_twist_01_l, loc=<-6.2083 0.0279 5.6817>, rot=<0 0.0007 0.0191 -0.9998>}
 deform_upperarm_up_l {parent=deform_upperarm_up_base_l, loc=<0 0 0>, rot=<0 0 0 1>}
 clavicle_r {parent=spine_05, loc=<5.4831 -0.1053 3.2946>, rot=<0.7289 0.0465 -0.6829 0.009>}
 upperarm_r {parent=clavicle_r, loc=<-9.8658 1.2136 0.4105>, rot=<0.0164 -0.3245 -0.0269 -0.9454>}
 upperarm_twist_01_r {parent=upperarm_r, loc=<-7.2135 0.1134 0>, rot=<0 0 0 -1>}
 deform_upperarm_bk_base_r {parent=upperarm_twist_01_r, loc=<2.1328 6.2124 -0.4113>, rot=<0 0.0007 0.0191 -0.9998>}
 deform_upperarm_bk_r {parent=deform_upperarm_bk_base_r, loc=<0 0 0>, rot=<0 0 0 1>}
 deform_upperarm_fr_base_r {parent=upperarm_twist_01_r, loc=<3.0358 -8.1024 0.4104>, rot=<0 0.0007 0.0191 -0.9998>}
 deform_upperarm_fr_r {parent=deform_upperarm_fr_base_r, loc=<0 0 0>, rot=<0 0 0 1>}
 deform_upperarm_up_base_r {parent=upperarm_twist_01_r, loc=<6.2079 -0.0279 -5.6814>, rot=<0 0.0007 0.0191 -0.9998>}
 deform_upperarm_up_r {parent=deform_upperarm_up_base_r, loc=<0 0 0>, rot=<0 0 0 -1>}
 MSTR_hood_dyn {parent=spine_05, loc=<12.5552 -5.2932 0>, rot=<0 0 0.8689 -0.495>}
 hood_dyn {parent=MSTR_hood_dyn, loc=<0 0 0>, rot=<0 0 0 1>}
 hood_deform_base_m {parent=hood_dyn, loc=<-0.6194 1.1557 0>, rot=<0 0 0.8726 0.4884>}
 hood_deform_m {parent=hood_deform_base_m, loc=<0 0 0>, rot=<0 0 0 1>}
 joint1 {parent=hood_deform_m, loc=<7.4377 -1.5262 0>, rot=<0.6463 -0.2869 0.2869 -0.6463>}
 hood_deform_correct {parent=hood_deform_m, loc=<7.4377 -1.5262 0>, rot=<0.6463 -0.2869 0.2869 -0.6463>}
 hood_pnt_mid {parent=hood_dyn, loc=<0 12 0>, rot=<0 0 0 1>}
 hood_up_base_m {parent=hood_dyn, loc=<-0.6194 1.1557 0>, rot=<0 0 0.8726 0.4884>}
 hood_up_m {parent=hood_up_base_m, loc=<0 0 0>, rot=<0 0 0 1>}
 hood_up_deform_base {parent=hood_up_m, loc=<0.5 -0.5 0>, rot=<0.6174 -0.3448 0.3448 -0.6174>}
 hood_up_deform {parent=hood_up_deform_base, loc=<0 0 0>, rot=<0 0 0 1>}
 hood_up_correct_base {parent=hood_up_m, loc=<5.7132 -0.74 0>, rot=<0 0 0.6054 -0.7959>}
 hood_up_correct {parent=hood_up_correct_base, loc=<0 0 0>, rot=<0 0 0 1>}
 hood_down_base {parent=hood_dyn, loc=<4.2248 -1.1744 0>, rot=<0 0 0.7333 -0.6799>}
 hood_down {parent=hood_down_base, loc=<0 0 0>, rot=<0 0 0 1>}
 hood_deform_base_r {parent=spine_05, loc=<11.4543 -0.6513 6.193>, rot=<0.5875 0.2542 0.0152 -0.7681>}
 hood_deform_r {parent=hood_deform_base_r, loc=<0 0 0>, rot=<0 0 0 1>}
 hood_deform_base_l {parent=spine_05, loc=<11.454 -0.6513 -6.193>, rot=<-0.5875 -0.2542 0.0152 -0.7681>}
 hood_deform_l {parent=hood_deform_base_l, loc=<0 0 0>, rot=<0 0 0 1>}
 HeadBoneAim {parent=spine_05, loc=<23 -8 0>, rot=<0 0 0 1>}
 pocket_for_l_deform_base {parent=spine_04, loc=<3.2732 11.533 -9.8838>, rot=<-0.4613 -0.4923 -0.2723 -0.6861>}
 pocket_for_l_deform {parent=pocket_for_l_deform_base, loc=<0 0 0>, rot=<0 0 0 -1>}
 pocket_for_r_deform_base {parent=spine_04, loc=<3.2733 11.533 9.8838>, rot=<0.6861 0.2723 0.4923 0.4613>}
 pocket_for_r_deform {parent=pocket_for_r_deform_base, loc=<0 0 0>, rot=<0 0 0 1>}
]
