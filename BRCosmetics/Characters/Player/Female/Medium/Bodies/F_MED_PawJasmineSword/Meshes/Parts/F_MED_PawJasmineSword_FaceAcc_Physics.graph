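# Physics asset for the hood on the face accessory part of the F_MED_PawJasmineSword female player body cosmetic.
# A tiny capsule on the master hood bone MSTR_hood_dyn pairs with a larger, zero-length capsule, effectively a sphere, offset on hood_dyn.
# One constraint links hood_dyn to the master bone, allowing moderate swing on one axis while keeping the other swing and twist nearly locked, so motion is hinge-like.
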PhysicsAsset /BRCosmetics/Characters/Player/Female/Medium/Bodies/F_MED_PawJasmineSword/Meshes/Parts/F_MED_PawJasmineSword_FaceAcc_Physics
bodies {
 MSTR_hood_dyn [{shapeType=Capsule, length=0.505, radius=0.505, shapeName=MSTR_hood_dyn_capsule, rotation=<rot 0 -90 0>}]
 hood_dyn [{
   shapeType=Capsule
   length=0
   radius=3.9881
   shapeName=cap_open_capsule
   center=<5.3218 -0.0067 0.0243>
   rotation=<rot 3.2858 90.4141 -90.0004>
  }]
}
constraints [{bone1Name=hood_dyn, bone2Name=MSTR_hood_dyn, swing1LimitDegrees=35, swing2LimitDegrees=3, twistLimitDegrees=1}]
