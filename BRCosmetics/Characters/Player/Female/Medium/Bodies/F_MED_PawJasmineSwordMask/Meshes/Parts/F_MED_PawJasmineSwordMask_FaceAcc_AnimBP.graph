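# Post-copy Anim Blueprint adding secondary motion to the PawJasmineSwordMask character's face accessory.
# It copies the attached parent's pose, runs RigidBody simulation based at spine_05, then a Control Rig, reusing PawJasmineSword's physics and rig assets.
# Each update sets gravity override and derives external force from spine_05-to-root offset change, scaled by velocity along the mesh's right axis.
# Two "Hide Element" material parameters determine hood state, which disables the simulation when on.
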AnimBlueprint /BRCosmetics/Characters/Player/Female/Medium/Bodies/F_MED_PawJasmineSwordMask/Meshes/Parts/F_MED_PawJasmineSwordMask_FaceAcc_AnimBP
target_skeleton=&/F_MED_PawJasmineSwordMask_FaceAcc_Skeleton
graphs [
 AnimGraph {nodes [
   n1 = RigidBody!(
    ComponentPose=Local To Component(LocalPose=Copy Pose From Mesh() {bUseAttachedParent=true})
    bAlphaBoolEnabled:bool=NOT Boolean(A=Get isHoodOn() -> (isHoodOn:bool))
    OverrideWorldGravity:Vector=Get Gravity Override()
    ExternalForce=Get delta_spine() -> (delta_spine:Vector)
   ) -> (Pose) {
    alphaInputType=Bool
    bOverrideWorldGravity=true
    overridePhysicsAsset=/BRCosmetics/Characters/Player/Female/Medium/Bodies/F_MED_PawJasmineSword/Meshes/Parts/F_MED_PawJasmineSword_FaceAcc_Physics
    simulationSpace=BaseBoneSpace
    alphaBoolBlend {blendOption=Linear}
    baseBoneRef {boneName=spine_05}
    componentAppliedLinearAccClamp=<0 0 0>
    simSpaceSettings {
     maxAngularAcceleration=5
     maxAngularVelocity=5
     maxLinearAcceleration=70
     maxLinearVelocity=150
     worldAlpha=1
     externalLinearDragV=<0.1 0.1 0.1>
    }
   }
   n2 = Control Rig(Source=Component To Local(ComponentPose=@n1)) {
    alphaBoolBlend {blendOption=Linear}
    controlRigAssetReference {blueprintRigClass=/BRCosmetics/Characters/Player/Female/Medium/Bodies/F_MED_PawJasmineSword/Meshes/Parts/F_MED_PawJasmineSword_FaceAcc_CtrlRig.F_MED_PawJasmineSword_FaceAcc_CtrlRig_C}
   }
   Output Pose(Result=@n2) {layerGroup=None}
  ]}
 EventGraph {nodes [
   Event BlueprintUpdateAnimation()
   > n1 = Gravity Override State Function DEPRECATED(
    CharacterPartAnimInstance=Self-Reference()
    GameTimeInSeconds=GetGameTimeInSeconds()
    GravityOverrideParams=Get Gravity Override Params()
    RBAN Params=Get RBAN Params()
   )
   > Set Gravity Override(Gravity Override=@n1)
   > n2 = Sequence!(execute) -> (then_0, then_1, then_2:Exec)
   n3 = Get material() -> (material:Material Instance Dynamic Object Reference)
   > n4 = Get Scalar Parameter Value!(execute, self:Material Instance Dynamic Object Reference=@n3, ParameterName:Name=Hide Element 01) -> (then, ReturnValue:float) {is=title K2Node_CallFunction}
   > n5 = Get Scalar Parameter Value(self=@n3, ParameterName=Hide Element 02)
   > Set isHoodOn(isHoodOn:bool=AND Boolean(A="Equal (Float)"(A=@n5), B="Equal (Float)"(A=@n4, B=1))) -> (Output_Get:bool)
   n6 = Get PawnVelocity()
   n7 = Get World Rotation!(self:Scene Component Object Reference=Get SkelMeshPart() -> (SkelMeshPart:Skeletal Mesh Component Object Reference)) -> (ReturnValue:Rotator) {is=title K2Node_CallFunction}
   n8 = float * float(
    A=Vector Length(A=Get PawnVelocity())
    B=Dot Product!(
     A:Vector=Normalize!(A:Vector=@n6, Tolerance:float=0.0001) -> (ReturnValue:Vector) {is=title K2Node_CallFunction}
     B:Vector=GetRightVector!(InRot:Rotator=@n7) -> (ReturnValue:Vector)
    ) -> (ReturnValue:double)
   ) {is=title K2Node_PromotableOperator}
   n9 = GetDeltaSeconds!(self:Anim Instance Object Reference) -> (ReturnValue:float)
   n10 = vector - vector(
    A=GetSocketLocation!(self:Scene Component Object Reference=Get SkelMeshPart(), InSocketName:Name=spine_05) -> (ReturnValue:Vector)
    B=GetSocketLocation(self=Get SkelMeshPart(), InSocketName=root)
   )
   n11 = vector - vector(A=@n10, B=Get prev_delta_spine() -> (prev_delta_spine:Vector))
   Set delta_spine(
    execute=@n2.then_1
    delta_spine:Vector=vector * vector!(
     A=vector / vector(A:Vector=@n11, B:float=@n9) -> (ReturnValue:Vector)
     B=Get delta_mult() -> (delta_mult:double)
     C:double=MapRangeClamped(Value=@n8, InRangeA=0, InRangeB=200)
    ) -> (ReturnValue=~)
   ) -> (Output_Get:Vector)
   > Set prev_delta_spine(prev_delta_spine:Vector=@n10) -> (Output_Get:Vector)
   n12 = Event BlueprintInitializeAnimation()
   n13 = Set SkelMeshPart(execute=[@n12.then, @n15.then], SkelMeshPart:Skeletal Mesh Component Object Reference=GetPartSkeletalMeshComponent()) -> (Output_Get:Skeletal Mesh Component Object Reference)
   > n14 = Get Dynamic Material Instance(SkeletalMeshComponent=@n13)
   > Set material(material:Material Instance Dynamic Object Reference=@n14) -> (Output_Get:Material Instance Dynamic Object Reference)
   n15 = Event OnCharacterCustomizationLoaded()
   GetForwardVector!(InRot:Rotator=@n7) -> (ReturnValue:Vector)
  ]}
]
variables [
 delta_mult {type=number}
 delta_spine {type=Vector}
 Gravity Override {type=Vector}
 Gravity Override Params {type=GravityOverrideParamsStruct}
 isHoodOn {type=boolean}
 material {type=MaterialInstanceDynamic}
 prev_delta_spine {type=Vector}
 RBAN Params {
  type=RBANParamsStruct
  default {
   dBNO {simSpaceSettings {
     maxAngularAcceleration=1000
     maxAngularVelocity=3
     maxLinearAcceleration=300
     maxLinearVelocity=700
     velocityScaleZ=0
     externalLinearDragV=<0 0.5 0>
    }}
   dBNO_Carried {simSpaceSettings {
     maxAngularAcceleration=1000
     maxAngularVelocity=3
     maxLinearAcceleration=300
     maxLinearVelocity=700
     velocityScaleZ=0
     externalLinearDragV=<0 0.5 0>
    }}
   emoteOrMelee {simSpaceSettings {
     maxAngularAcceleration=1000
     maxAngularVelocity=3
     maxLinearAcceleration=300
     maxLinearVelocity=700
     velocityScaleZ=0
     externalLinearDragV=<0 0.5 0>
    }}
   front_End {simSpaceSettings {
     maxAngularAcceleration=1000
     maxAngularVelocity=3
     maxLinearAcceleration=300
     maxLinearVelocity=700
     velocityScaleZ=0
     externalLinearDragV=<0 0.5 0>
    }}
   onGround_Crouch_Moving {simSpaceSettings {
     maxAngularAcceleration=1000
     maxAngularVelocity=3
     maxLinearAcceleration=300
     maxLinearVelocity=700
     velocityScaleZ=0
     externalLinearDragV=<0 0.5 0>
    }}
   onGround_Falling {simSpaceSettings {
     maxAngularAcceleration=1000
     maxAngularVelocity=3
     maxLinearAcceleration=300
     maxLinearVelocity=700
     velocityScaleZ=0
     externalLinearDragV=<0 0.5 0>
    }}
   onGround_Moving {simSpaceSettings {
     maxAngularAcceleration=1000
     maxAngularVelocity=3
     maxLinearAcceleration=300
     maxLinearVelocity=700
     velocityScaleZ=0
     externalLinearDragV=<0 0.5 0>
    }}
   onGround_Standing {simSpaceSettings {
     maxAngularAcceleration=1000
     maxAngularVelocity=3
     maxLinearAcceleration=300
     maxLinearVelocity=700
     velocityScaleZ=0
     externalLinearDragV=<0 0.5 0>
    }}
   rail_Grinding {simSpaceSettings {
     maxAngularAcceleration=1000
     maxAngularVelocity=3
     maxLinearAcceleration=300
     maxLinearVelocity=700
     velocityScaleZ=0
     worldAlpha=0.1
     externalLinearDragV=<0 0.5 0>
    }}
   riding_Moving {simSpaceSettings {
     maxAngularAcceleration=1000
     maxAngularVelocity=3
     maxLinearAcceleration=300
     maxLinearVelocity=700
     velocityScaleZ=0
     externalLinearDragV=<0 0.5 0>
    }}
   riding_Moving_Targeting {simSpaceSettings {
     maxAngularAcceleration=1000
     maxAngularVelocity=3
     maxLinearAcceleration=300
     maxLinearVelocity=700
     velocityScaleZ=0
     externalLinearDragV=<0 0.5 0>
    }}
   riding_Standing {simSpaceSettings {
     maxAngularAcceleration=1000
     maxAngularVelocity=3
     maxLinearAcceleration=300
     maxLinearVelocity=700
     velocityScaleZ=0
     externalLinearDragV=<0 0.5 0>
    }}
   skydive_DivingDown {simSpaceSettings {
     maxAngularAcceleration=1000
     maxAngularVelocity=3
     maxLinearAcceleration=300
     maxLinearVelocity=700
     velocityScaleZ=0
     externalLinearDragV=<0 0.5 0>
    }}
   skydive_FloatingDown {
    componentLinearAccScale=<0 0.5 0>
    componentLinearVelScale=<0 0.5 0>
    simSpaceSettings {
     maxAngularAcceleration=1000
     maxAngularVelocity=3
     maxLinearAcceleration=300
     maxLinearVelocity=700
     velocityScaleZ=0
     externalLinearDragV=<0 0.5 0>
    }
   }
   skydive_GoingUp {simSpaceSettings {
     maxAngularAcceleration=1000
     maxAngularVelocity=3
     maxLinearAcceleration=300
     maxLinearVelocity=700
     velocityScaleZ=0
     externalLinearDragV=<0 0.5 0>
    }}
   skydive_Parachuting {
    componentLinearAccScale=<0 0.1 0.3>
    componentLinearVelScale=<0 0.1 0.3>
    simSpaceSettings {
     maxAngularAcceleration=1000
     maxAngularVelocity=3
     maxLinearAcceleration=300
     maxLinearVelocity=700
     velocityScaleZ=0
     externalLinearDragV=<0 0.5 0>
    }
   }
   skydive_Surfing {simSpaceSettings {
     maxAngularAcceleration=1000
     maxAngularVelocity=3
     maxLinearAcceleration=300
     maxLinearVelocity=700
     velocityScaleZ=0
     externalLinearDragV=<0 0.5 0>
    }}
   swimming {simSpaceSettings {
     maxAngularAcceleration=1000
     maxAngularVelocity=3
     maxLinearAcceleration=300
     maxLinearVelocity=700
     velocityScaleZ=0
     externalLinearDragV=<0 0.5 0>
    }}
   tactical_Sprint {simSpaceSettings {
     maxAngularAcceleration=1000
     maxAngularVelocity=3
     maxLinearAcceleration=300
     maxLinearVelocity=700
     velocityScaleZ=0
     externalLinearDragV=<0 0.5 0>
    }}
   ziplining {simSpaceSettings {
     maxAngularAcceleration=1000
     maxAngularVelocity=3
     maxLinearAcceleration=300
     maxLinearVelocity=700
     velocityScaleZ=0
     externalLinearDragV=<0 0.5 0>
    }}
  }
 }
 SkelMeshPart {type=SkeletalMeshComponent}
]
